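# Single Layer Water surface material from the Water plugin, named for panning textures and usable on instanced static meshes.
# WaterTextureSurface takes WaterAttributes and water time, with near and far scale and normal inputs.
# The water output derives scattering and absorption from color parameters, and drives color behind water with SubUV-animated caustics fed by refraction.
# A Gerstner-wave caustics blend is also present, with no visible consumer.
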Material /Water/Materials/WaterSurface/Water_PanningTextures
graph {
 properties {
  bTangentSpaceNormal=false
  bUseMaterialAttributes=true
  bUsedWithInstancedStaticMeshes=true
  bUsedWithWater=true
  pixelDepthOffsetMode=PDOM_Legacy
  refractionMethod=RM_PixelNormalOffset
  shadingModel=MSM_SingleLayerWater
  usedShadingModels=SingleLayerWater
 }
 nodes [
  LinearInterpolate(
   B=MaterialFunctionCall!(Refraction=ScalarParameter() {defaultValue=1, parameterName=Ref}, Caustics Brightness=ScalarParameter() {defaultValue=1, parameterName=B}) {materialFunction=/Water/Caustics/Materials/Functions/WaterCaustics_GerstnerWaves}.Caustics
   Alpha=ScalarParameter() {defaultValue=0.5, parameterName=Caustics Contrast}
  ) {constA=1}
  MaterialFunctionCall!(
   Wave Attributes=MaterialFunctionCall() {materialFunction=/Water/Materials/Layers/WaterAttributes}.Result
   Time=MaterialFunctionCall() {materialFunction=/Water/Materials/Functions/GetWaterTime}.WaterTime
   Distant Scale
   Near Scale
   Blend Alpha
   Speed
   Far Normal Strength
   Near Normal Strength
   Far Displacement
  ) {materialFunction=/Water/Materials/Functions/WaterTextureSurface}
  n1 = VectorParameter() {parameterName=Scattering, defaultValue=<rgba 1 1 1 0>}
  n2 = MaterialFunctionCall() {materialFunction=&WaterAttributes}
  SingleLayerWaterMaterialOutput!(
   ScatteringCoefficients=Divide(A=Multiply(A=@n1.RGB, B=@n1.RGB)) {constB=1000}
   AbsorptionCoefficients=Divide(A=VectorParameter() {parameterName=Absorption Coefficient, defaultValue=<rgba 0.6 0.05 0.005 0>}.RGB) {constB=100}
   PhaseG=ScalarParameter() {defaultValue=0.1, parameterName=Anisotropy}
   ColorScaleBehindWater=MaterialFunctionCall!(
    Refraction=GetMaterialAttributes(None=@n2.Result) {attributeGetTypes [#1]}.Refraction
    Tiling=ScalarParameter() {defaultValue=4096, parameterName=Caustics Scale}
    Caustics Brightness
   ) {materialFunction=/Water/Caustics/Materials/Functions/WaterCaustics_SubUVAnimation}.Caustics
  )
 ]
}
